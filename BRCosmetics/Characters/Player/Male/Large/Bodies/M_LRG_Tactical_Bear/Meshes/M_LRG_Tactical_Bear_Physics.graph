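# Physics asset for the large male Tactical Bear body cosmetic, covering only its dynamic tail and flap bones.
# Small capsule bodies on two parent bones anchor the chains: a two-segment tail and separate top and bottom flaps.
# Angular limits keep the tail fairly stiff, while both flaps hang from the shared flap parent with much wider swing and little twist.
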PhysicsAsset /BRCosmetics/Characters/Player/Male/Large/Bodies/M_LRG_Tactical_Bear/Meshes/M_LRG_Tactical_Bear_Physics
bodies {
 dyn_flap_bot_01 [{shapeType=Capsule, length=5.5, radius=3.75, center=<3 0 0>, rotation=<rot 90 0 0>}]
 dyn_flap_top_01 [{shapeType=Capsule, length=3.75, radius=1.75, center=<2 0 0>, rotation=<rot 90 0 0>}]
 dyn_parent_flaps [{shapeType=Capsule, length=0.505, radius=0.505, rotation=<rot 0 -90 0>}]
 dyn_parent_tail [{shapeType=Capsule, length=0.505, radius=0.505, rotation=<rot 0 -90 0>}]
 dyn_tail_01 [{shapeType=Capsule, length=0, radius=6.75, center=<1 0 0>, rotation=<rot 90 0 0>}]
 dyn_tail_02 [{shapeType=Capsule, length=3, radius=3.75, center=<1.5 0 0>, rotation=<rot 90 0 0>}]
}
constraints [
 {bone1Name=dyn_tail_01, bone2Name=dyn_parent_tail, swing1LimitDegrees=20, swing2LimitDegrees=20, twistLimitDegrees=8}
 {bone1Name=dyn_tail_02, bone2Name=dyn_tail_01, swing1LimitDegrees=20, swing2LimitDegrees=20, twistLimitDegrees=8}
 {bone1Name=dyn_flap_top_01, bone2Name=dyn_parent_flaps, swing1LimitDegrees=67.2857, swing2LimitDegrees=55.2857, twistLimitDegrees=5}
 {bone1Name=dyn_flap_bot_01, bone2Name=dyn_parent_flaps, swing1LimitDegrees=45, swing2LimitDegrees=45, twistLimitDegrees=5}
]
