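# Post-pose AnimBlueprint for the large male Tactical Bear body cosmetic, adding secondary motion to straps, flaps, pouches, scarf and other gear.
# It copies the attached parent mesh's pose, then runs a long component-space chain.
# Weighted multi-bone constraints position dyn_parent bones, and spring-driven, angle-limited AnimDynamics nodes simulate gear bones.
# Calf and thigh rotation alphas blend additive corrective transforms on knee straps and glutes.
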AnimBlueprint /BRCosmetics/Characters/Player/Male/Large/Bodies/M_LRG_Tactical_Bear/Meshes/M_LRG_Tactical_Bear_AnimBP
target_skeleton=&/M_LRG_Tactical_Bear_Skeleton
graphs [
 AnimGraph {nodes [
   Anim Dynamics - Bone: dyn_arm_flap_02_l(Alpha=1, ExternalForce="0,0,0") {
    angularDampingOverride=1
    angularSpringConstant=1500
    bAngularSpring=true
    bGravityOverrideInSimSpace=true
    bOverrideAngularDamping=true
    bUseGravityOverride=true
    gravityScale=0.35
    simulationSpace=RootRelative
    alphaBoolBlend {blendOption=Linear}
    boundBone {boneName=dyn_arm_flap_02_l}
    componentAppliedLinearAccClamp=<3000 3000 3500>
    componentLinearAccScale=<0.3 0.3 0.3>
    componentLinearVelScale=<0.3 0.3 1.5>
    gravityOverride {z=-980}
    physicsBodyDefinitions [{
      sphereCollisionRadius=0
      boundBone {boneName=dyn_arm_flap_02_l}
      boxExtents=<20 30 10>
      constraintSetup {angularLimitsMax=<3 5 30>, angularLimitsMin=<-2 -5 -30>, angularTarget=<0 0 0>}
      localJointOffset=<1 0 0>
     }]
   }
   "Constraint - dyn_parent_tail (PPPPPPPP)"!(
    ComponentPose
    Alpha=1
    ConstraintWeights_0=0.0028
    ConstraintWeights_1:float=0.0946
    ConstraintWeights_2:float=0.0939
    ConstraintWeights_3:float=0.0007
    ConstraintWeights_4:float=0.0055
    ConstraintWeights_5:float=0.0025
    ConstraintWeights_6:float=0.7139
    ConstraintWeights_7:float=0.0861
   ) -> (Pose) {
    alphaBoolBlend {blendOption=Linear}
    boneToModify {boneName=dyn_parent_tail}
    constraintSetup [
     {targetBone {boneName=thigh_r}}
     {targetBone {boneName=thigh_twist_01_r}}
     {targetBone {boneName=spine_01}}
     {targetBone {boneName=spine_02}}
     {targetBone {boneName=spine_03}}
     {targetBone {boneName=thigh_l}}
     {targetBone {boneName=pelvis}}
     {targetBone {boneName=thigh_twist_01_l}}
    ]
    constraintWeights [1,1,1,1,1,1,1,1]
   }
   n1 = "Constraint - dyn_parent_knee_strap_out_l (PPPPP)"!(
    ComponentPose=Local To Component(LocalPose=Copy Pose From Mesh() {bUseAttachedParent=true})
    Alpha=1
    ConstraintWeights_0=0.1424
    ConstraintWeights_1:float=0.7677
    ConstraintWeights_2:float=0.0121
    ConstraintWeights_3:float=0.0142
    ConstraintWeights_4:float=0.0637
   ) -> (Pose) {
    alphaBoolBlend {blendOption=Linear}
    boneToModify {boneName=dyn_parent_knee_strap_out_l}
    constraintSetup [
     {targetBone {boneName=calf_l}}
     {targetBone {boneName=dyn_parent_knee_strap_out_l}}
     {targetBone {boneName=thigh_twist_01_l}}
     {targetBone {boneName=calf_twist_02_l}}
     {targetBone {boneName=thigh_l}}
    ]
    constraintWeights [1,1,1,1,1]
   }
   n2 = "Constraint - dyn_parent_knee_strap_in_l (PPPPP)"!(
    ComponentPose=@n1
    Alpha=1
    ConstraintWeights_0=0.2754
    ConstraintWeights_1:float=0.518
    ConstraintWeights_2:float=0.0255
    ConstraintWeights_3:float=0.0407
    ConstraintWeights_4:float=0.1404
   ) -> (Pose) {
    alphaBoolBlend {blendOption=Linear}
    boneToModify {boneName=dyn_parent_knee_strap_in_l}
    constraintSetup [
     {targetBone {boneName=calf_l}}
     {targetBone {boneName=dyn_parent_knee_strap_in_l}}
     {targetBone {boneName=thigh_twist_01_l}}
     {targetBone {boneName=calf_twist_02_l}}
     {targetBone {boneName=thigh_l}}
    ]
    constraintWeights [1,1,1,1,1]
   }
   n3 = Get CalfLeftRotationAlpha() -> (CalfLeftRotationAlpha:float)
   n4 = "Transform (Modify) Bone - Bone: dyn_parent_knee_strap_out_l"(ComponentPose=@n2, Alpha=@n3, Translation="-2,-0.000,-0.000", Rotation="0.000,-10,0.000", Scale="0.75,0.75,0.75") {
    rotationMode=BMM_Additive
    rotationSpace=BCS_BoneSpace
    scaleMode=BMM_Additive
    scaleSpace=BCS_BoneSpace
    translationMode=BMM_Additive
    translationSpace=BCS_BoneSpace
    alphaBoolBlend {blendOption=Linear}
    boneToModify {boneName=dyn_parent_knee_strap_out_l}
    rotation {yaw=-10}
    scale=<0.75 0.75 0.75>
    translation {x=-2}
   }
   n5 = "Transform (Modify) Bone - Bone: dyn_parent_knee_strap_in_l"(ComponentPose=@n4, Alpha=@n3, Translation="2.5,-0.000,-0.000", Rotation="0,0,0", Scale="0.75,0.75,0.75") {
    rotationMode=BMM_Additive
    rotationSpace=BCS_BoneSpace
    scaleMode=BMM_Additive
    scaleSpace=BCS_BoneSpace
    translationMode=BMM_Additive
    translationSpace=BCS_BoneSpace
    alphaBoolBlend {blendOption=Linear}
    boneToModify {boneName=dyn_parent_knee_strap_in_l}
    scale=<0.75 0.75 0.75>
    translation {x=2.5}
   }
   n6 = Anim Dynamics - Bone: dyn_knee_flap_out_l(ComponentPose=@n5, Alpha=1, ExternalForce="0,0,0") {
    angularDampingOverride=1
    angularSpringConstant=1500
    bAngularSpring=true
    bGravityOverrideInSimSpace=true
    bOverrideAngularDamping=true
    bUseGravityOverride=true
    gravityScale=0.35
    simulationSpace=RootRelative
    alphaBoolBlend {blendOption=Linear}
    boundBone {boneName=dyn_knee_flap_out_l}
    componentAppliedLinearAccClamp=<3000 3000 3500>
    componentLinearAccScale=<0.3 0.3 0.3>
    componentLinearVelScale=<0.3 0.3 1.5>
    gravityOverride {z=-980}
    physicsBodyDefinitions [{
      sphereCollisionRadius=0
      boundBone {boneName=dyn_knee_flap_out_l}
      boxExtents=<15 15 20>
      constraintSetup {angularLimitsMax=<5 5 30>, angularLimitsMin=<-5 -25 -30>, angularTarget=<0 0 0>}
      localJointOffset=<0 4 0>
     }]
   }
   n7 = Anim Dynamics - Bone: dyn_knee_flap_in_l(ComponentPose=@n6, Alpha=0.65, ExternalForce="0,0,0") {
    angularDampingOverride=1
    angularSpringConstant=1500
    bAngularSpring=true
    bGravityOverrideInSimSpace=true
    bOverrideAngularDamping=true
    bUseGravityOverride=true
    gravityScale=0.35
    simulationSpace=RootRelative
    alphaBoolBlend {blendOption=Linear}
    boundBone {boneName=dyn_knee_flap_in_l}
    componentAppliedLinearAccClamp=<3000 3000 3000>
    componentLinearAccScale=<0.3 0.3 0.3>
    componentLinearVelScale=<0.3 0.3 2>
    gravityOverride {z=-980}
    physicsBodyDefinitions [{
      sphereCollisionRadius=0
      boundBone {boneName=dyn_knee_flap_in_l}
      boxExtents=<15 10 10>
      constraintSetup {angularLimitsMax=<1 5 30>, angularLimitsMin=<-1 -45 -30>, angularTarget=<0 0 0>}
      localJointOffset=<0 -8 -1>
     }]
   }
   n8 = "Constraint - dyn_parent_knee_strap_out_r (PPPPP)"!(
    ComponentPose=@n7
    Alpha=1
    ConstraintWeights_0=0.0646
    ConstraintWeights_1:float=0.0142
    ConstraintWeights_2:float=0.7677
    ConstraintWeights_3:float=0.0125
    ConstraintWeights_4:float=0.141
   ) -> (Pose) {
    alphaBoolBlend {blendOption=Linear}
    boneToModify {boneName=dyn_parent_knee_strap_out_r}
    constraintSetup [
     {targetBone {boneName=thigh_r}}
     {targetBone {boneName=calf_twist_02_r}}
     {targetBone {boneName=dyn_parent_knee_strap_out_r}}
     {targetBone {boneName=thigh_twist_01_r}}
     {targetBone {boneName=calf_r}}
    ]
    constraintWeights [1,1,1,1,1]
   }
   n9 = "Constraint - dyn_parent_knee_strap_in_r (PPPPP)"!(
    ComponentPose=@n8
    Alpha=1
    ConstraintWeights_0=0.1363
    ConstraintWeights_1:float=0.0425
    ConstraintWeights_2:float=0.518
    ConstraintWeights_3:float=0.023
    ConstraintWeights_4:float=0.2802
   ) -> (Pose) {
    alphaBoolBlend {blendOption=Linear}
    boneToModify {boneName=dyn_parent_knee_strap_in_r}
    constraintSetup [
     {targetBone {boneName=thigh_r}}
     {targetBone {boneName=calf_twist_02_r}}
     {targetBone {boneName=dyn_parent_knee_strap_in_r}}
     {targetBone {boneName=thigh_twist_01_r}}
     {targetBone {boneName=calf_r}}
    ]
    constraintWeights [1,1,1,1,1]
   }
   n10 = Get CalfRightRotationAlpha() -> (CalfRightRotationAlpha:float)
   n11 = "Transform (Modify) Bone - Bone: dyn_parent_knee_strap_out_r"(ComponentPose=@n9, Alpha=@n10, Translation="2,-0.000,-0.000", Rotation="0.000,-10,0.000", Scale="0.75,0.75,0.75") {
    rotationMode=BMM_Additive
    rotationSpace=BCS_BoneSpace
    scaleMode=BMM_Additive
    scaleSpace=BCS_BoneSpace
    translationMode=BMM_Additive
    translationSpace=BCS_BoneSpace
    alphaBoolBlend {blendOption=Linear}
    boneToModify {boneName=dyn_parent_knee_strap_out_r}
    rotation {yaw=-10}
    scale=<0.75 0.75 0.75>
    translation {x=2}
   }
   n12 = "Transform (Modify) Bone - Bone: dyn_parent_knee_strap_in_r"(ComponentPose=@n11, Alpha=@n10, Translation="-2.5,-0.000,-0.000", Rotation="0,0,0", Scale="0.75,0.75,0.75") {
    rotationMode=BMM_Additive
    rotationSpace=BCS_BoneSpace
    scaleMode=BMM_Additive
    scaleSpace=BCS_BoneSpace
    translationMode=BMM_Additive
    translationSpace=BCS_BoneSpace
    alphaBoolBlend {blendOption=Linear}
    boneToModify {boneName=dyn_parent_knee_strap_in_r}
    scale=<0.75 0.75 0.75>
    translation {x=-2.5}
   }
   n13 = Anim Dynamics - Bone: dyn_knee_flap_out_r(ComponentPose=@n12, Alpha=1, ExternalForce="0,0,0") {
    angularDampingOverride=1
    angularSpringConstant=1500
    bAngularSpring=true
    bGravityOverrideInSimSpace=true
    bOverrideAngularDamping=true
    bUseGravityOverride=true
    gravityScale=0.35
    simulationSpace=RootRelative
    alphaBoolBlend {blendOption=Linear}
    boundBone {boneName=dyn_knee_flap_out_r}
    componentAppliedLinearAccClamp=<3000 3000 3500>
    componentLinearAccScale=<0.3 0.3 0.3>
    componentLinearVelScale=<0.3 0.3 1.5>
    gravityOverride {z=-980}
    physicsBodyDefinitions [{
      sphereCollisionRadius=0
      boundBone {boneName=dyn_knee_flap_out_r}
      boxExtents=<15 10 10>
      constraintSetup {angularLimitsMax=<5 5 30>, angularLimitsMin=<-5 -15 -30>, angularTarget=<0 0 0>}
      localJointOffset=<0 -6 0>
     }]
   }
   n14 = Anim Dynamics - Bone: dyn_knee_flap_in_r(ComponentPose=@n13, Alpha=0.65, ExternalForce="0,0,0") {
    angularDampingOverride=1
    angularSpringConstant=1500
    bAngularSpring=true
    bGravityOverrideInSimSpace=true
    bOverrideAngularDamping=true
    bUseGravityOverride=true
    gravityScale=0.35
    simulationSpace=RootRelative
    alphaBoolBlend {blendOption=Linear}
    boundBone {boneName=dyn_knee_flap_in_r}
    componentAppliedLinearAccClamp=<3000 3000 3000>
    componentLinearAccScale=<0.3 0.3 0.3>
    componentLinearVelScale=<0.3 0.3 2>
    gravityOverride {z=-980}
    physicsBodyDefinitions [{
      sphereCollisionRadius=0
      boundBone {boneName=dyn_knee_flap_in_r}
      boxExtents=<15 15 20>
      constraintSetup {angularLimitsMax=<1 5 30>, angularLimitsMin=<-1 -45 -30>, angularTarget=<0 0 0>}
      localJointOffset=<0 8 0>
     }]
   }
   n15 = "Constraint - dyn_parent_calf_flapA_out_r (PP)"!(ComponentPose=@n14, Alpha=1, ConstraintWeights_0=0.4475, ConstraintWeights_1:float=0.5525) -> (Pose) {
    alphaBoolBlend {blendOption=Linear}
    boneToModify {boneName=dyn_parent_calf_flapA_out_r}
    constraintSetup [{targetBone {boneName=calf_twist_02_r}}, {targetBone {boneName=calf_r}}]
    constraintWeights [1,1]
   }
   n16 = "Constraint - dyn_parent_calf_flapA_out_l (PPPP)"!(
    ComponentPose=@n15
    Alpha=1
    ConstraintWeights_0=0.5362
    ConstraintWeights_1:float=0.4309
    ConstraintWeights_2:float=0.0001
    ConstraintWeights_3:float=0.0328
   ) -> (Pose) {
    alphaBoolBlend {blendOption=Linear}
    boneToModify {boneName=dyn_parent_calf_flapA_out_l}
    constraintSetup [
     {targetBone {boneName=calf_l}}
     {targetBone {boneName=calf_twist_02_l}}
     {targetBone {boneName=thigh_l}}
     {targetBone {boneName=dyn_calf_flapA_out_l}}
    ]
    constraintWeights [1,1,1,1]
   }
   n17 = "Constraint - dyn_parent_calf_flapA_in_r (PPP)"!(ComponentPose=@n16, Alpha=1, ConstraintWeights_0=0.0103, ConstraintWeights_1:float=0.2305, ConstraintWeights_2:float=0.7592) -> (Pose) {
    alphaBoolBlend {blendOption=Linear}
    boneToModify {boneName=dyn_parent_calf_flapA_in_r}
    constraintSetup [{targetBone {boneName=thigh_r}}, {targetBone {boneName=calf_twist_02_r}}, {targetBone {boneName=calf_r}}]
    constraintWeights [1,1,1]
   }
   n18 = "Constraint - dyn_parent_calf_flapA_in_l (PPPP)"!(
    ComponentPose=@n17
    Alpha=1
    ConstraintWeights_0=0.6998
    ConstraintWeights_1:float=0.232
    ConstraintWeights_2:float=0.0049
    ConstraintWeights_3:float=0.0633
   ) -> (Pose) {
    alphaBoolBlend {blendOption=Linear}
    boneToModify {boneName=dyn_parent_calf_flapA_in_l}
    constraintSetup [
     {targetBone {boneName=calf_l}}
     {targetBone {boneName=calf_twist_02_l}}
     {targetBone {boneName=thigh_l}}
     {targetBone {boneName=dyn_calf_flapA_in_l}}
    ]
    constraintWeights [1,1,1,1]
   }
   n19 = "Constraint - dyn_parent_calf_flapB_out_r (PP)"!(ComponentPose=@n18, Alpha=1, ConstraintWeights_0=0.3941, ConstraintWeights_1:float=0.6059) -> (Pose) {
    alphaBoolBlend {blendOption=Linear}
    boneToModify {boneName=dyn_parent_calf_flapB_out_r}
    constraintSetup [{targetBone {boneName=calf_twist_02_r}}, {targetBone {boneName=foot_r}}]
    constraintWeights [1,1]
   }
   n20 = "Constraint - dyn_parent_calf_flapB_out_l (PPPP)"!(
    ComponentPose=@n19
    Alpha=1
    ConstraintWeights_0=0.0951
    ConstraintWeights_1:float=0.5326
    ConstraintWeights_2:float=0
    ConstraintWeights_3:float=0.3723
   ) -> (Pose) {
    alphaBoolBlend {blendOption=Linear}
    boneToModify {boneName=dyn_parent_calf_flapB_out_l}
    constraintSetup [
     {targetBone {boneName=dyn_calf_flapB_out_l}}
     {targetBone {boneName=foot_l}}
     {targetBone {boneName=ball_l}}
     {targetBone {boneName=calf_twist_02_l}}
    ]
    constraintWeights [1,1,1,1]
   }
   n21 = "Constraint - dyn_parent_calf_flapB_in_r (PPP)"!(ComponentPose=@n20, Alpha=1, ConstraintWeights_0=0.2292, ConstraintWeights_1:float=0.7584, ConstraintWeights_2:float=0.0124) -> (Pose) {
    alphaBoolBlend {blendOption=Linear}
    boneToModify {boneName=dyn_parent_calf_flapB_in_r}
    constraintSetup [{targetBone {boneName=calf_twist_02_r}}, {targetBone {boneName=foot_r}}, {targetBone {boneName=calf_r}}]
    constraintWeights [1,1,1]
   }
   n22 = "Constraint - dyn_parent_calf_flapB_in_l (PPPP)"!(
    ComponentPose=@n21
    Alpha=1
    ConstraintWeights_0=0.0113
    ConstraintWeights_1:float=0.5864
    ConstraintWeights_2:float=0.2097
    ConstraintWeights_3:float=0.1926
   ) -> (Pose) {
    alphaBoolBlend {blendOption=Linear}
    boneToModify {boneName=dyn_parent_calf_flapB_in_l}
    constraintSetup [
     {targetBone {boneName=calf_l}}
     {targetBone {boneName=foot_l}}
     {targetBone {boneName=calf_twist_02_l}}
     {targetBone {boneName=dyn_calf_flapB_in_l}}
    ]
    constraintWeights [1,1,1,1]
   }
   n23 = "Constraint - dyn_parent_foot_flap_out_r (P)"!(ComponentPose=@n22, Alpha=1, ConstraintWeights_0=1, ConstraintWeights_1:float=0) -> (Pose) {
    alphaBoolBlend {blendOption=Linear}
    boneToModify {boneName=dyn_parent_foot_flap_out_r}
    constraintSetup [{targetBone {boneName=foot_r}}]
    constraintWeights [1,0]
   }
   n24 = "Constraint - dyn_parent_foot_flap_out_l (PPPP)"!(
    ComponentPose=@n23
    Alpha=1
    ConstraintWeights_0=0.0595
    ConstraintWeights_1:float=0.939
    ConstraintWeights_2:float=0.0015
    ConstraintWeights_3:float=0.0001
   ) -> (Pose) {
    alphaBoolBlend {blendOption=Linear}
    boneToModify {boneName=dyn_parent_foot_flap_out_l}
    constraintSetup [
     {targetBone {boneName=dyn_foot_flap_out_l}}
     {targetBone {boneName=foot_l}}
     {targetBone {boneName=ball_l}}
     {targetBone {boneName=calf_twist_02_l}}
    ]
    constraintWeights [1,1,1,1]
   }
   n25 = "Constraint - dyn_parent_foot_flap_in_r (PP)"!(ComponentPose=@n24, Alpha=1, ConstraintWeights_0=0.9531, ConstraintWeights_1:float=0.0469) -> (Pose) {
    alphaBoolBlend {blendOption=Linear}
    boneToModify {boneName=dyn_parent_foot_flap_in_r}
    constraintSetup [{targetBone {boneName=foot_r}}, {targetBone {boneName=calf_r}}]
    constraintWeights [1,1]
   }
   n26 = "Constraint - dyn_parent_foot_flap_in_l (PPP)"!(ComponentPose=@n25, Alpha=1, ConstraintWeights_0=0.0449, ConstraintWeights_1:float=0.9251, ConstraintWeights_2:float=0.03) -> (Pose) {
    alphaBoolBlend {blendOption=Linear}
    boneToModify {boneName=dyn_parent_foot_flap_in_l}
    constraintSetup [{targetBone {boneName=calf_l}}, {targetBone {boneName=foot_l}}, {targetBone {boneName=dyn_foot_flap_in_l}}]
    constraintWeights [1,1,1]
   }
   n27 = Anim Dynamics - Bone: dyn_calf_flapA_out_r(ComponentPose=@n26, Alpha=1, ExternalForce="0,0,0") {
    angularDampingOverride=1
    angularSpringConstant=1500
    bAngularSpring=true
    bGravityOverrideInSimSpace=true
    bOverrideAngularDamping=true
    gravityScale=0.35
    simulationSpace=RootRelative
    alphaBoolBlend {blendOption=Linear}
    boundBone {boneName=dyn_calf_flapA_out_r}
    componentAppliedLinearAccClamp=<3000 3000 3000>
    componentLinearAccScale=<0.3 0.3 0.3>
    componentLinearVelScale=<3 0.3 0.3>
    physicsBodyDefinitions [{
      sphereCollisionRadius=0
      boundBone {boneName=dyn_calf_flapA_out_r}
      boxExtents=<15 10 10>
      constraintSetup {angularLimitsMax=<0 0 40>, angularLimitsMin=<0 -30 -40>, angularTarget=<0 0 0>}
      localJointOffset=<0 -5 -1>
     }]
   }
   n28 = Anim Dynamics - Bone: dyn_calf_flapA_out_l(ComponentPose=@n27, Alpha=0.6, ExternalForce="0,0,0") {
    angularDampingOverride=1
    angularSpringConstant=1500
    bAngularSpring=true
    bGravityOverrideInSimSpace=true
    bOverrideAngularDamping=true
    gravityScale=0.35
    simulationSpace=RootRelative
    alphaBoolBlend {blendOption=Linear}
    boundBone {boneName=dyn_calf_flapA_out_l}
    componentAppliedLinearAccClamp=<3000 3000 3000>
    componentLinearAccScale=<0.3 0.3 0.3>
    componentLinearVelScale=<3 0.3 0.3>
    physicsBodyDefinitions [{
      sphereCollisionRadius=0
      boundBone {boneName=dyn_calf_flapA_out_l}
      boxExtents=<20 10 10>
      constraintSetup {angularLimitsMax=<0 5 40>, angularLimitsMin=<0 -35 -40>, angularTarget=<0 0 0>}
      localJointOffset=<0 5 0>
     }]
   }
   n29 = Anim Dynamics - Bone: dyn_calf_flapA_in_r(ComponentPose=@n28, Alpha=0.9, ExternalForce="0,0,0") {
    angularDampingOverride=1
    angularSpringConstant=1500
    bAngularSpring=true
    bGravityOverrideInSimSpace=true
    bOverrideAngularDamping=true
    gravityScale=0.35
    simulationSpace=RootRelative
    alphaBoolBlend {blendOption=Linear}
    boundBone {boneName=dyn_calf_flapA_in_r}
    componentAppliedLinearAccClamp=<3000 3000 3000>
    componentLinearAccScale=<0.3 0.3 0.3>
    componentLinearVelScale=<3 0.3 0.3>
    physicsBodyDefinitions [{
      sphereCollisionRadius=0
      boundBone {boneName=dyn_calf_flapA_in_r}
      boxExtents=<20 8 10>
      constraintSetup {angularLimitsMax=<0 5 25>, angularLimitsMin=<0 -30 -40>, angularTarget=<0 0 0>}
      localJointOffset=<0 5 0>
     }]
   }
   n30 = Anim Dynamics - Bone: dyn_calf_flapA_in_l(ComponentPose=@n29, Alpha=0.75, ExternalForce="0,0,0") {
    angularDampingOverride=1
    angularSpringConstant=1500
    bAngularSpring=true
    bGravityOverrideInSimSpace=true
    bOverrideAngularDamping=true
    gravityScale=0.35
    simulationSpace=RootRelative
    alphaBoolBlend {blendOption=Linear}
    boundBone {boneName=dyn_calf_flapA_in_l}
    componentAppliedLinearAccClamp=<3000 3000 3000>
    componentLinearAccScale=<0.3 0.3 0.3>
    componentLinearVelScale=<3 0.3 0.3>
    physicsBodyDefinitions [{
      sphereCollisionRadius=0
      boundBone {boneName=dyn_calf_flapA_in_l}
      boxExtents=<15 15 10>
      constraintSetup {angularLimitsMax=<0 5 0>, angularLimitsMin=<0 -35 -40>, angularTarget=<0 0 0>}
      localJointOffset=<0 -8 0>
     }]
   }
   n31 = Anim Dynamics - Bone: dyn_calf_flapB_out_r(ComponentPose=@n30, Alpha=1, ExternalForce="0,0,0") {
    angularDampingOverride=1
    angularSpringConstant=1500
    bAngularSpring=true
    bGravityOverrideInSimSpace=true
    bOverrideAngularDamping=true
    gravityScale=0.35
    simulationSpace=RootRelative
    alphaBoolBlend {blendOption=Linear}
    boundBone {boneName=dyn_calf_flapB_out_r}
    componentAppliedLinearAccClamp=<3000 3000 3000>
    componentLinearAccScale=<0.3 0.3 0.3>
    componentLinearVelScale=<3 0.3 0.3>
    physicsBodyDefinitions [{
      sphereCollisionRadius=0
      boundBone {boneName=dyn_calf_flapB_out_r}
      boxExtents=<15 10 10>
      constraintSetup {angularLimitsMax=<0 5 60>, angularLimitsMin=<0 -35 -40>, angularTarget=<0 0 0>}
      localJointOffset=<0 -4 0>
     }]
   }
   n32 = Anim Dynamics - Bone: dyn_calf_flapB_out_l(ComponentPose=@n31, Alpha=1, ExternalForce="0,0,0") {
    angularDampingOverride=1
    angularSpringConstant=1500
    bAngularSpring=true
    bGravityOverrideInSimSpace=true
    bOverrideAngularDamping=true
    gravityScale=0.35
    simulationSpace=RootRelative
    alphaBoolBlend {blendOption=Linear}
    boundBone {boneName=dyn_calf_flapB_out_l}
    componentAppliedLinearAccClamp=<3000 3000 3000>
    componentLinearAccScale=<0.3 0.3 0.3>
    componentLinearVelScale=<3 0.3 0.3>
    physicsBodyDefinitions [{
      sphereCollisionRadius=0
      boundBone {boneName=dyn_calf_flapB_out_l}
      boxExtents=<12 15 10>
      constraintSetup {angularLimitsMax=<0 50 60>, angularLimitsMin=<0 -60 -40>, angularTarget=<0 0 0>}
      localJointOffset=<0 3 0>
     }]
   }
   n33 = Anim Dynamics - Bone: dyn_calf_flapB_in_r(ComponentPose=@n32, Alpha=0.7, ExternalForce="0,0,0") {
    angularDampingOverride=1
    angularSpringConstant=1500
    bAngularSpring=true
    bGravityOverrideInSimSpace=true
    bOverrideAngularDamping=true
    gravityScale=0.35
    simulationSpace=RootRelative
    alphaBoolBlend {blendOption=Linear}
    boundBone {boneName=dyn_calf_flapB_in_r}
    componentAppliedLinearAccClamp=<3000 3000 3000>
    componentLinearAccScale=<0.3 0.3 0.3>
    componentLinearVelScale=<3 0.3 0.3>
    physicsBodyDefinitions [{
      sphereCollisionRadius=0
      boundBone {boneName=dyn_calf_flapB_in_r}
      boxExtents=<20 10 10>
      constraintSetup {angularLimitsMax=<0 5 55>, angularLimitsMin=<0 -35 -40>, angularTarget=<0 0 0>}
      localJointOffset=<0 5 -1>
     }]
   }
   n34 = Anim Dynamics - Bone: dyn_calf_flapB_in_l(ComponentPose=@n33, Alpha=1, ExternalForce="0,0,0") {
    angularDampingOverride=1
    angularSpringConstant=1500
    bAngularSpring=true
    bGravityOverrideInSimSpace=true
    bOverrideAngularDamping=true
    gravityScale=0.35
    simulationSpace=RootRelative
    alphaBoolBlend {blendOption=Linear}
    boundBone {boneName=dyn_calf_flapB_in_l}
    componentAppliedLinearAccClamp=<3000 3000 3000>
    componentLinearAccScale=<0.3 0.3 0.3>
    componentLinearVelScale=<3 0.3 0.3>
    physicsBodyDefinitions [{
      sphereCollisionRadius=0
      boundBone {boneName=dyn_calf_flapB_in_l}
      boxExtents=<20 10 10>
      constraintSetup {angularLimitsMax=<0 45 65>, angularLimitsMin=<0 -35 -40>, angularTarget=<0 0 0>}
      localJointOffset=<0 -4 -0.2>
     }]
   }
   n35 = Anim Dynamics - Bone: dyn_foot_flap_out_r(ComponentPose=@n34, Alpha=0.8, ExternalForce="0,0,0") {
    angularDampingOverride=1
    angularSpringConstant=1500
    bAngularSpring=true
    bGravityOverrideInSimSpace=true
    bOverrideAngularDamping=true
    gravityScale=0.35
    simulationSpace=RootRelative
    alphaBoolBlend {blendOption=Linear}
    boundBone {boneName=dyn_foot_flap_out_r}
    componentAppliedLinearAccClamp=<3000 3000 3000>
    componentLinearAccScale=<0.3 0.3 0.3>
    componentLinearVelScale=<3 0.3 0.3>
    physicsBodyDefinitions [{
      sphereCollisionRadius=0
      boundBone {boneName=dyn_foot_flap_out_r}
      boxExtents=<5 10 5>
      constraintSetup {angularLimitsMax=<0 60 35>, angularLimitsMin=<0 -45 -40>, angularTarget=<0 0 0>}
      localJointOffset=<-1 -0.1 0>
     }]
   }
   n36 = Anim Dynamics - Bone: dyn_foot_flap_out_l(ComponentPose=@n35, Alpha=0.55, ExternalForce="0,0,0") {
    angularDampingOverride=1
    angularSpringConstant=1500
    bAngularSpring=true
    bGravityOverrideInSimSpace=true
    bOverrideAngularDamping=true
    gravityScale=0.35
    simulationSpace=RootRelative
    alphaBoolBlend {blendOption=Linear}
    boundBone {boneName=dyn_foot_flap_out_l}
    componentAppliedLinearAccClamp=<3000 3000 3000>
    componentLinearAccScale=<0.3 0.3 0.3>
    componentLinearVelScale=<3 0.3 0.3>
    physicsBodyDefinitions [{
      sphereCollisionRadius=0
      boundBone {boneName=dyn_foot_flap_out_l}
      boxExtents=<5 10 5>
      constraintSetup {angularLimitsMax=<0 60 35>, angularLimitsMin=<0 -45 -40>, angularTarget=<0 0 0>}
      localJointOffset=<1 -0.1 0>
     }]
   }
   n37 = Anim Dynamics - Bone: dyn_foot_flap_in_r(ComponentPose=@n36, Alpha=0.9, ExternalForce="0,0,0") {
    angularDampingOverride=1
    angularSpringConstant=1500
    bAngularSpring=true
    bGravityOverrideInSimSpace=true
    bOverrideAngularDamping=true
    gravityScale=0.35
    simulationSpace=RootRelative
    alphaBoolBlend {blendOption=Linear}
    boundBone {boneName=dyn_foot_flap_in_r}
    componentAppliedLinearAccClamp=<3000 3000 3000>
    componentLinearAccScale=<0.3 0.3 0.3>
    componentLinearVelScale=<3 0.3 0.3>
    physicsBodyDefinitions [{
      sphereCollisionRadius=0
      boundBone {boneName=dyn_foot_flap_in_r}
      boxExtents=<8 5 5>
      constraintSetup {angularLimitsMax=<0 45 35>, angularLimitsMin=<0 -55 -40>, angularTarget=<0 0 0>}
      localJointOffset=<-1.5 -0.1 0>
     }]
   }
   n38 = Anim Dynamics - Bone: dyn_foot_flap_in_l(ComponentPose=@n37, Alpha=1, ExternalForce="0,0,0") {
    angularDampingOverride=1
    angularSpringConstant=1500
    bAngularSpring=true
    bGravityOverrideInSimSpace=true
    bOverrideAngularDamping=true
    gravityScale=0.35
    simulationSpace=RootRelative
    alphaBoolBlend {blendOption=Linear}
    boundBone {boneName=dyn_foot_flap_in_l}
    componentAppliedLinearAccClamp=<3000 3000 3000>
    componentLinearAccScale=<0.3 0.3 0.3>
    componentLinearVelScale=<3 0.3 0.3>
    physicsBodyDefinitions [{
      sphereCollisionRadius=0
      boundBone {boneName=dyn_foot_flap_in_l}
      boxExtents=<5 6 5>
      constraintSetup {angularLimitsMax=<0 65 25>, angularLimitsMin=<0 -66 -40>, angularTarget=<0 0 0>}
      localJointOffset=<1 -0.2 0>
     }]
   }
   n39 = "Transform (Modify) Bone - Bone: dyn_glute_l"(
    ComponentPose=@n38
    Alpha=Get ThighLeftRotationAlpha() -> (ThighLeftRotationAlpha:float)
    Translation="-5,-1,-1"
    Rotation="0.000,-0.001,-70"
    Scale="1.2,1.21,1.2"
   ) {
    rotationMode=BMM_Additive
    rotationSpace=BCS_BoneSpace
    scaleMode=BMM_Additive
    scaleSpace=BCS_BoneSpace
    translationMode=BMM_Additive
    translationSpace=BCS_BoneSpace
    alphaBoolBlend {blendOption=Linear}
    boneToModify {boneName=dyn_glute_l}
    rotation {roll=-70, yaw=-0.001}
    scale=<1.2 1.21 1.2>
    translation=<-5 -1 -1>
   }
   n40 = "Transform (Modify) Bone - Bone: dyn_glute_r"(
    ComponentPose=@n39
    Alpha=Get ThighRightRotationAlpha() -> (ThighRightRotationAlpha:float)
    Translation="5.000,-1,-1"
    Rotation="0,0,-70"
    Scale="1.2,1.2,1.2"
   ) {
    rotationMode=BMM_Additive
    rotationSpace=BCS_BoneSpace
    scaleMode=BMM_Additive
    scaleSpace=BCS_BoneSpace
    translationMode=BMM_Additive
    translationSpace=BCS_BoneSpace
    alphaBoolBlend {blendOption=Linear}
    boneToModify {boneName=dyn_glute_r}
    rotation {roll=-70}
    scale=<1.2 1.2 1.2>
    translation=<5 -1 -1>
   }
   n41 = "Constraint - dyn_parent_hoop (PPPPPPP)"!(
    ComponentPose=@n40
    Alpha=1
    ConstraintWeights_0=0.0345
    ConstraintWeights_1:float=0.1102
    ConstraintWeights_2:float=0.2427
    ConstraintWeights_3:float=0.0824
    ConstraintWeights_4:float=0.26
    ConstraintWeights_5:float=0.2677
    ConstraintWeights_6:float=0.0024
   ) -> (Pose) {
    alphaBoolBlend {blendOption=Linear}
    boneToModify {boneName=dyn_parent_hoop}
    constraintSetup [
     {targetBone {boneName=clavicle_l}}
     {targetBone {boneName=spine_01}}
     {targetBone {boneName=spine_02}}
     {targetBone {boneName=spine_03}}
     {targetBone {boneName=spine_04}}
     {targetBone {boneName=spine_05}}
     {targetBone {boneName=upperarm_twist_01_l}}
    ]
    constraintWeights [1,1,1,1,1,1,1]
   }
   n42 = Anim Dynamics - Bone: dyn_hoop_01(ComponentPose=@n41, Alpha=1, ExternalForce="0,0,0") {
    angularDampingOverride=1
    angularSpringConstant=1500
    bAngularSpring=true
    bGravityOverrideInSimSpace=true
    bOverrideAngularDamping=true
    bUseGravityOverride=true
    gravityScale=0.35
    simulationSpace=RootRelative
    alphaBoolBlend {blendOption=Linear}
    boundBone {boneName=dyn_hoop_01}
    componentAppliedLinearAccClamp=<3000 3000 3500>
    componentLinearAccScale=<0.3 0.3 0.3>
    componentLinearVelScale=<0.3 0.3 1.5>
    gravityOverride {z=-980}
    physicsBodyDefinitions [{
      sphereCollisionRadius=0
      boundBone {boneName=dyn_hoop_01}
      boxExtents=<15 10 10>
      constraintSetup {angularLimitsMax=<5 0 5>, angularLimitsMin=<-5 -45 -5>, angularTarget=<0 0 0>}
      localJointOffset=<4 0 0>
     }]
   }
   n43 = "Constraint - dyn_parent_scarf (PPPPPPPP)"!(
    ComponentPose=@n42
    Alpha=1
    ConstraintWeights_0=0.0128
    ConstraintWeights_1:float=0.0148
    ConstraintWeights_2:float=0.0589
    ConstraintWeights_3:float=0.1404
    ConstraintWeights_4:float=0.4358
    ConstraintWeights_5:float=0.3094
    ConstraintWeights_6:float=0.0203
    ConstraintWeights_7:float=0.0076
   ) -> (Pose) {
    alphaBoolBlend {blendOption=Linear}
    boneToModify {boneName=dyn_parent_scarf}
    constraintSetup [
     {targetBone {boneName=clavicle_l}}
     {targetBone {boneName=spine_01}}
     {targetBone {boneName=spine_02}}
     {targetBone {boneName=spine_04}}
     {targetBone {boneName=spine_05}}
     {targetBone {boneName=dyn_scarf_01}}
     {targetBone {boneName=clavicle_r}}
     {targetBone {boneName=neck_02}}
    ]
    constraintWeights [1,1,1,1,1,1,1,1]
   }
   n44 = Anim Dynamics - Bone: dyn_scarf_01(ComponentPose=@n43, Alpha=1, ExternalForce="0,0,0") {
    angularDampingOverride=1
    angularSpringConstant=1500
    bAngularSpring=true
    bGravityOverrideInSimSpace=true
    bOverrideAngularDamping=true
    bUseGravityOverride=true
    gravityScale=0.35
    simulationSpace=RootRelative
    alphaBoolBlend {blendOption=Linear}
    boundBone {boneName=dyn_scarf_01}
    componentAppliedLinearAccClamp=<3000 3000 3000>
    componentLinearAccScale=<0.3 0.3 0.3>
    componentLinearVelScale=<0.3 0.3 2>
    gravityOverride {z=-980}
    physicsBodyDefinitions [{
      sphereCollisionRadius=0
      boundBone {boneName=dyn_scarf_01}
      boxExtents=<15 20 10>
      constraintSetup {angularLimitsMax=<1 0 20>, angularLimitsMin=<-1 -45 -20>, angularTarget=<0 0 0>}
      localJointOffset=<4 0 0>
     }]
   }
   n45 = "Constraint - dyn_parent_hip_pouch_r (PPPPPPP)"!(
    ComponentPose=@n44
    Alpha=1
    ConstraintWeights_0=0.645
    ConstraintWeights_1:float=0.023
    ConstraintWeights_2:float=0.0345
    ConstraintWeights_3:float=0.0034
    ConstraintWeights_4:float=0
    ConstraintWeights_5:float=0
    ConstraintWeights_6:float=0.2941
   ) -> (Pose) {
    alphaBoolBlend {blendOption=Linear}
    boneToModify {boneName=dyn_parent_hip_pouch_r}
    constraintSetup [
     {targetBone {boneName=spine_01}}
     {targetBone {boneName=spine_02}}
     {targetBone {boneName=spine_03}}
     {targetBone {boneName=spine_04}}
     {targetBone {boneName=spine_05}}
     {targetBone {boneName=clavicle_r}}
     {targetBone {boneName=pelvis}}
    ]
    constraintWeights [1,1,1,1,1,1,1]
   }
   n46 = Anim Dynamics - Bone: dyn_hip_pouch_01_r(ComponentPose=@n45, Alpha=1, ExternalForce="0,0,0") {
    angularDampingOverride=1
    angularSpringConstant=1500
    bAngularSpring=true
    bGravityOverrideInSimSpace=true
    bOverrideAngularDamping=true
    bUseGravityOverride=true
    gravityScale=0.35
    simulationSpace=RootRelative
    alphaBoolBlend {blendOption=Linear}
    boundBone {boneName=dyn_hip_pouch_01_r}
    componentAppliedLinearAccClamp=<3000 3000 3000>
    componentLinearAccScale=<0.3 0.3 0.3>
    componentLinearVelScale=<0.3 0.3 2>
    gravityOverride {z=-980}
    physicsBodyDefinitions [{
      sphereCollisionRadius=0
      boundBone {boneName=dyn_hip_pouch_01_r}
      boxExtents=<20 10 10>
      constraintSetup {angularLimitsMax=<5 0 5>, angularLimitsMin=<-5 -45 -5>, angularTarget=<0 0 0>}
      localJointOffset=<4 0 0>
     }]
   }
   n47 = Anim Dynamics - Bone: dyn_hip_pouch_flap_01_r(ComponentPose=@n46, Alpha=0.8, ExternalForce="0,0,0") {
    angularDampingOverride=1
    angularSpringConstant=1500
    bAngularSpring=true
    bGravityOverrideInSimSpace=true
    bOverrideAngularDamping=true
    bUseGravityOverride=true
    gravityScale=0.35
    simulationSpace=RootRelative
    alphaBoolBlend {blendOption=Linear}
    boundBone {boneName=dyn_hip_pouch_flap_01_r}
    componentAppliedLinearAccClamp=<3000 3000 3000>
    componentLinearAccScale=<0.3 0.3 0.3>
    componentLinearVelScale=<0.3 0.3 2>
    gravityOverride {z=-980}
    physicsBodyDefinitions [{
      boundBone {boneName=dyn_hip_pouch_flap_01_r}
      boxExtents=<20 10 10>
      constraintSetup {angularLimitsMax=<5 0 5>, angularLimitsMin=<-5 -45 -5>, angularTarget=<0 0 0>}
      localJointOffset=<4 0 0>
     }]
    sphericalLimits [{limitRadius=10, drivingBone {boneName=thigh_r}, sphereLocalOffset=<10 10 0>}]
   }
   n48 = "Constraint - dyn_parent_hip_pouch_l (PPPP)"!(
    ComponentPose=@n47
    Alpha=1
    ConstraintWeights_0=0.2921
    ConstraintWeights_1:float=0.5262
    ConstraintWeights_2:float=0.04
    ConstraintWeights_3:float=0.1417
   ) -> (Pose) {
    alphaBoolBlend {blendOption=Linear}
    boneToModify {boneName=dyn_parent_hip_pouch_l}
    constraintSetup [{targetBone {boneName=pelvis}}, {targetBone {boneName=thigh_twist_01_l}}, {targetBone {boneName=thigh_l}}, {targetBone {boneName=spine_01}}]
    constraintWeights [1,1,1,1]
   }
   n49 = Anim Dynamics - Bone: dyn_hip_pouch_01_l(ComponentPose=@n48, Alpha=1, ExternalForce="0,0,0") {
    angularDampingOverride=1
    angularSpringConstant=1500
    bAngularSpring=true
    bGravityOverrideInSimSpace=true
    bOverrideAngularDamping=true
    bUseGravityOverride=true
    gravityScale=0.35
    simulationSpace=RootRelative
    alphaBoolBlend {blendOption=Linear}
    boundBone {boneName=dyn_hip_pouch_01_l}
    componentAppliedLinearAccClamp=<3000 3000 3500>
    componentLinearAccScale=<0.3 0.3 0.3>
    componentLinearVelScale=<0.3 0.3 1.5>
    gravityOverride {z=-980}
    physicsBodyDefinitions [{
      sphereCollisionRadius=0
      boundBone {boneName=dyn_hip_pouch_01_l}
      boxExtents=<20 20 10>
      constraintSetup {angularLimitsMax=<5 0 5>, angularLimitsMin=<-5 -45 -5>, angularTarget=<0 0 0>}
      localJointOffset=<4 0 0>
     }]
   }
   n50 = Anim Dynamics - Bone: dyn_hip_pouch_flap_01_l(ComponentPose=@n49, Alpha=0.8, ExternalForce="0,0,0") {
    angularDampingOverride=1
    angularSpringConstant=1500
    bAngularSpring=true
    bGravityOverrideInSimSpace=true
    bOverrideAngularDamping=true
    bUseGravityOverride=true
    gravityScale=0.35
    simulationSpace=RootRelative
    alphaBoolBlend {blendOption=Linear}
    boundBone {boneName=dyn_hip_pouch_flap_01_l}
    componentAppliedLinearAccClamp=<3000 3000 3000>
    componentLinearAccScale=<0.3 0.3 0.3>
    componentLinearVelScale=<0.3 0.3 2>
    gravityOverride {z=-980}
    physicsBodyDefinitions [{
      boundBone {boneName=dyn_hip_pouch_flap_01_l}
      boxExtents=<20 20 10>
      constraintSetup {angularLimitsMax=<5 0 5>, angularLimitsMin=<-5 -45 -5>, angularTarget=<0 0 0>}
      localJointOffset=<4 0 0>
     }]
   }
   n51 = "Constraint - dyn_parent_arm_flap (PPPP)"!(
    ComponentPose=@n50
    Alpha=1
    ConstraintWeights_0=0.2021
    ConstraintWeights_1:float=0.1033
    ConstraintWeights_2:float=0.1432
    ConstraintWeights_3:float=0.5513
   ) -> (Pose) {
    alphaBoolBlend {blendOption=Linear}
    boneToModify {boneName=dyn_parent_arm_flap}
    constraintSetup [
     {targetBone {boneName=lowerarm_l}}
     {targetBone {boneName=lowerarm_twist_01_l}}
     {targetBone {boneName=upperarm_l}}
     {targetBone {boneName=lowerarm_twist_02_l}}
    ]
    constraintWeights [1,1,1,1]
   }
   n52 = Anim Dynamics - Bone: dyn_arm_flap_01_l(ComponentPose=@n51, Alpha=0.65, ExternalForce="0,0,0") {
    angularDampingOverride=1
    angularSpringConstant=1500
    bAngularSpring=true
    bGravityOverrideInSimSpace=true
    bOverrideAngularDamping=true
    bUseGravityOverride=true
    gravityScale=0.35
    simulationSpace=RootRelative
    alphaBoolBlend {blendOption=Linear}
    boundBone {boneName=dyn_arm_flap_01_l}
    componentAppliedLinearAccClamp=<3000 3000 3000>
    componentLinearAccScale=<0.3 0.3 0.3>
    componentLinearVelScale=<0.3 0.3 2>
    gravityOverride {z=-980}
    physicsBodyDefinitions [{
      sphereCollisionRadius=0
      boundBone {boneName=dyn_arm_flap_01_l}
      boxExtents=<20 20 10>
      constraintSetup {angularLimitsMax=<1 5 30>, angularLimitsMin=<-1 -5 -30>, angularTarget=<0 0 0>}
      localJointOffset=<4 0 0>
     }]
   }
   n53 = "Constraint - dyn_parent_upperarm_flap (PPPPPPP)"!(
    ComponentPose=@n52
    Alpha=1
    ConstraintWeights_0=0.1263
    ConstraintWeights_1:float=0.1788
    ConstraintWeights_2:float=0.0118
    ConstraintWeights_3:float=0.3437
    ConstraintWeights_4:float=0.0234
    ConstraintWeights_5:float=0.3058
    ConstraintWeights_6:float=0.0102
   ) -> (Pose) {
    alphaBoolBlend {blendOption=Linear}
    boneToModify {boneName=dyn_parent_upperarm_flap}
    constraintSetup [
     {targetBone {boneName=clavicle_l}}
     {targetBone {boneName=upperarm_twist_02_l}}
     {targetBone {boneName=spine_04}}
     {targetBone {boneName=upperarm_twist_01_l}}
     {targetBone {boneName=lowerarm_l}}
     {targetBone {boneName=upperarm_l}}
     {targetBone {boneName=lowerarm_twist_02_l}}
    ]
    constraintWeights [1,1,1,1,1,1,1]
   }
   n54 = Anim Dynamics - Bone: dyn_upperarm_flap_01_l(ComponentPose=@n53, Alpha=1, ExternalForce="0,0,0") {
    angularDampingOverride=1
    angularSpringConstant=1500
    bAngularSpring=true
    bGravityOverrideInSimSpace=true
    bOverrideAngularDamping=true
    bUseGravityOverride=true
    gravityScale=0.35
    simulationSpace=RootRelative
    alphaBoolBlend {blendOption=Linear}
    boundBone {boneName=dyn_upperarm_flap_01_l}
    componentAppliedLinearAccClamp=<3000 3000 3500>
    componentLinearAccScale=<0.3 0.3 0.3>
    componentLinearVelScale=<0.3 0.3 1.5>
    gravityOverride {z=-980}
    physicsBodyDefinitions [{
      sphereCollisionRadius=0
      boundBone {boneName=dyn_upperarm_flap_01_l}
      boxExtents=<20 20 10>
      constraintSetup {angularLimitsMax=<5 20 5>, angularLimitsMin=<-5 -45 -5>, angularTarget=<0 0 0>}
      localJointOffset=<4 0 0>
     }]
   }
   n55 = "Constraint - dyn_parent_pocket_01 (PPPPPPPP)"!(
    ComponentPose=@n54
    Alpha=1
    ConstraintWeights_0=0.0086
    ConstraintWeights_1:float=0.1992
    ConstraintWeights_2:float=0.2712
    ConstraintWeights_3:float=0.2128
    ConstraintWeights_4:float=0.1957
    ConstraintWeights_5:float=0.1039
    ConstraintWeights_6:float=0.006
    ConstraintWeights_7:float=0.0028
   ) -> (Pose) {
    alphaBoolBlend {blendOption=Linear}
    boneToModify {boneName=dyn_parent_pocket_01}
    constraintSetup [
     {targetBone {boneName=clavicle_l}}
     {targetBone {boneName=spine_01}}
     {targetBone {boneName=spine_02}}
     {targetBone {boneName=spine_03}}
     {targetBone {boneName=spine_04}}
     {targetBone {boneName=spine_05}}
     {targetBone {boneName=upperarm_twist_01_l}}
     {targetBone {boneName=pelvis}}
    ]
    constraintWeights [1,1,1,1,1,1,1,1]
   }
   n56 = Anim Dynamics - Bone: dyn_pocket_01(ComponentPose=@n55, Alpha=1, ExternalForce="0,0,0") {
    angularDampingOverride=1
    angularSpringConstant=1500
    bAngularSpring=true
    bGravityOverrideInSimSpace=true
    bOverrideAngularDamping=true
    bUseGravityOverride=true
    gravityScale=0.35
    simulationSpace=RootRelative
    alphaBoolBlend {blendOption=Linear}
    boundBone {boneName=dyn_pocket_01}
    componentAppliedLinearAccClamp=<3000 3000 3000>
    componentLinearAccScale=<0.3 0.3 0.3>
    componentLinearVelScale=<0.3 0.3 2>
    gravityOverride {z=-980}
    physicsBodyDefinitions [{
      sphereCollisionRadius=0
      boundBone {boneName=dyn_pocket_01}
      boxExtents=<20 10 10>
      constraintSetup {angularLimitsMax=<5 0 5>, angularLimitsMin=<-5 -45 -5>, angularTarget=<0 0 0>}
      localJointOffset=<4 0 0>
     }]
   }
   n57 = "Constraint - dyn_parent_pocket_02 (PPPPPPPP)"!(
    ComponentPose=@n56
    Alpha=1
    ConstraintWeights_0=0.0031
    ConstraintWeights_1:float=0.1611
    ConstraintWeights_2:float=0.2772
    ConstraintWeights_3:float=0.1902
    ConstraintWeights_4:float=0.2207
    ConstraintWeights_5:float=0.1472
    ConstraintWeights_6:float=0.0002
    ConstraintWeights_7:float=0.0003
   ) -> (Pose) {
    alphaBoolBlend {blendOption=Linear}
    boneToModify {boneName=dyn_parent_pocket_02}
    constraintSetup [
     {targetBone {boneName=clavicle_l}}
     {targetBone {boneName=spine_01}}
     {targetBone {boneName=spine_02}}
     {targetBone {boneName=spine_03}}
     {targetBone {boneName=spine_04}}
     {targetBone {boneName=spine_05}}
     {targetBone {boneName=upperarm_twist_01_l}}
     {targetBone {boneName=pelvis}}
    ]
    constraintWeights [1,1,1,1,1,1,1,1]
   }
   n58 = Anim Dynamics - Bone: dyn_pocket_02(ComponentPose=@n57, Alpha=0.8, ExternalForce="0,0,0") {
    angularDampingOverride=1
    angularSpringConstant=1500
    bAngularSpring=true
    bGravityOverrideInSimSpace=true
    bOverrideAngularDamping=true
    bUseGravityOverride=true
    gravityScale=0.35
    simulationSpace=RootRelative
    alphaBoolBlend {blendOption=Linear}
    boundBone {boneName=dyn_pocket_02}
    componentAppliedLinearAccClamp=<3000 3000 3000>
    componentLinearAccScale=<0.3 0.3 0.3>
    componentLinearVelScale=<0.3 0.3 2>
    gravityOverride {z=-980}
    physicsBodyDefinitions [{
      boundBone {boneName=dyn_pocket_02}
      boxExtents=<20 20 10>
      constraintSetup {angularLimitsMax=<5 0 5>, angularLimitsMin=<-5 -45 -5>, angularTarget=<0 0 0>}
      localJointOffset=<4 0 0>
     }]
   }
   n59 = "Constraint - dyn_parent_pocket_03 (PPPPPPPPPP)"!(
    ComponentPose=@n58
    Alpha=1
    ConstraintWeights_0=0
    ConstraintWeights_1:float=0.0001
    ConstraintWeights_2:float=0.1593
    ConstraintWeights_3:float=0.2815
    ConstraintWeights_4:float=0.1903
    ConstraintWeights_5:float=0.2243
    ConstraintWeights_6:float=0.1441
    ConstraintWeights_7:float=0
    ConstraintWeights_8:float=0
    ConstraintWeights_9:float=0.0002
   ) -> (Pose) {
    alphaBoolBlend {blendOption=Linear}
    boneToModify {boneName=dyn_parent_pocket_03}
    constraintSetup [
     {targetBone {boneName=clavicle_l}}
     {targetBone {boneName=upperarm_twist_01_r}}
     {targetBone {boneName=spine_01}}
     {targetBone {boneName=spine_02}}
     {targetBone {boneName=spine_03}}
     {targetBone {boneName=spine_04}}
     {targetBone {boneName=spine_05}}
     {targetBone {boneName=upperarm_twist_01_l}}
     {targetBone {boneName=clavicle_r}}
     {targetBone {boneName=pelvis}}
    ]
    constraintWeights [1,1,1,1,1,1,1,1,1,1]
   }
   n60 = Anim Dynamics - Bone: dyn_pocket_03(ComponentPose=@n59, Alpha=1, ExternalForce="0,0,0") {
    angularDampingOverride=1
    angularSpringConstant=1500
    bAngularSpring=true
    bGravityOverrideInSimSpace=true
    bOverrideAngularDamping=true
    bUseGravityOverride=true
    gravityScale=0.35
    simulationSpace=RootRelative
    alphaBoolBlend {blendOption=Linear}
    boundBone {boneName=dyn_pocket_03}
    componentAppliedLinearAccClamp=<3000 3000 3500>
    componentLinearAccScale=<0.3 0.3 0.3>
    componentLinearVelScale=<0.3 0.3 1.5>
    gravityOverride {z=-980}
    physicsBodyDefinitions [{
      sphereCollisionRadius=0
      boundBone {boneName=dyn_pocket_03}
      boxExtents=<20 20 10>
      constraintSetup {angularLimitsMax=<5 0 5>, angularLimitsMin=<-5 -45 -5>, angularTarget=<0 0 0>}
      localJointOffset=<4 0 0>
     }]
   }
   n61 = "Constraint - dyn_parent_pocket_04 (PPPPPPPP)"!(
    ComponentPose=@n60
    Alpha=1
    ConstraintWeights_0=0.0054
    ConstraintWeights_1:float=0.1874
    ConstraintWeights_2:float=0.2802
    ConstraintWeights_3:float=0.2061
    ConstraintWeights_4:float=0.2105
    ConstraintWeights_5:float=0.1074
    ConstraintWeights_6:float=0.0009
    ConstraintWeights_7:float=0.0021
   ) -> (Pose) {
    alphaBoolBlend {blendOption=Linear}
    boneToModify {boneName=dyn_parent_pocket_04}
    constraintSetup [
     {targetBone {boneName=upperarm_twist_01_r}}
     {targetBone {boneName=spine_01}}
     {targetBone {boneName=spine_02}}
     {targetBone {boneName=spine_03}}
     {targetBone {boneName=spine_04}}
     {targetBone {boneName=spine_05}}
     {targetBone {boneName=clavicle_r}}
     {targetBone {boneName=pelvis}}
    ]
    constraintWeights [1,1,1,1,1,1,1,1]
   }
   n62 = Anim Dynamics - Bone: dyn_pocket_04(ComponentPose=@n61, Alpha=1, ExternalForce="0,0,0") {
    angularDampingOverride=1
    angularSpringConstant=1500
    bAngularSpring=true
    bGravityOverrideInSimSpace=true
    bOverrideAngularDamping=true
    bUseGravityOverride=true
    gravityScale=0.35
    simulationSpace=RootRelative
    alphaBoolBlend {blendOption=Linear}
    boundBone {boneName=dyn_pocket_04}
    componentAppliedLinearAccClamp=<3000 3000 3000>
    componentLinearAccScale=<0.3 0.3 0.3>
    componentLinearVelScale=<0.3 0.3 2>
    gravityOverride {z=-980}
    physicsBodyDefinitions [{
      sphereCollisionRadius=0
      boundBone {boneName=dyn_pocket_04}
      boxExtents=<15 20 10>
      constraintSetup {angularLimitsMax=<1 0 5>, angularLimitsMin=<-1 -45 -3>, angularTarget=<0 0 0>}
      localJointOffset=<4 0 0>
     }]
   }
   n63 = Flap Node - Bone: thigh_twist_offset_l!(ComponentPose:Component Space Pose Link Structure=@n62, Alpha:float=1) -> (Pose:Component Space Pose Link Structure) {
    is=title FortAnimGraphNode_Flap
    bUseClamp=true
    targetClampMin=-3
    targetComponent=RotationZ
    alphaBoolBlend {blendOption=Linear}
    sourceDrives [
     {
      bUseRange=false
      drivingCurve=null
      multiplier=-1
      rangeMax=1
      rangeMin=-1
      remappedMax=1
      remappedMin=0
      sourceComponent=RotationZ
      useQuaternion=true
      sourceBone {boneName=thigh_l}
     }
     {
      bUseRange=false
      drivingCurve=null
      multiplier=1
      rangeMax=1
      rangeMin=-1
      remappedMax=1
      remappedMin=0
      sourceComponent=RotationZ
      useQuaternion=true
      sourceBone {boneName=thigh_r}
     }
    ]
    targetBone {boneName=thigh_twist_offset_l}
   }
   n64 = Flap Node - Bone: thigh_twist_offset_r!(ComponentPose:Component Space Pose Link Structure=@n63, Alpha:float=1) -> (Pose:Component Space Pose Link Structure) {
    is=title FortAnimGraphNode_Flap
    bUseClamp=true
    targetClampMin=-3
    targetComponent=RotationZ
    alphaBoolBlend {blendOption=Linear}
    sourceDrives [
     {
      bUseRange=false
      drivingCurve=null
      multiplier=-1.2
      rangeMax=1
      rangeMin=-1
      remappedMax=1
      remappedMin=0
      sourceComponent=RotationZ
      useQuaternion=true
      sourceBone {boneName=thigh_r}
     }
     {
      bUseRange=false
      drivingCurve=null
      multiplier=1
      rangeMax=1
      rangeMin=-1
      remappedMax=1
      remappedMin=0
      sourceComponent=RotationZ
      useQuaternion=true
      sourceBone {boneName=thigh_l}
     }
    ]
    targetBone {boneName=thigh_twist_offset_r}
   }
   n65 = "Constraint - dyn_parent_pelvis_pack_l (PPPPPP)"!(
    ComponentPose=@n64
    Alpha=1
    ConstraintWeights_0=0.4413
    ConstraintWeights_1:float=0.1998
    ConstraintWeights_2:float=0.2212
    ConstraintWeights_3:float=0.0789
    ConstraintWeights_4:float=0.0243
    ConstraintWeights_5:float=0.0345
   ) -> (Pose) {
    alphaBoolBlend {blendOption=Linear}
    boneToModify {boneName=dyn_parent_pelvis_pack_l}
    constraintSetup [
     {targetBone {boneName=spine_01}}
     {targetBone {boneName=spine_02}}
     {targetBone {boneName=spine_03}}
     {targetBone {boneName=spine_04}}
     {targetBone {boneName=spine_05}}
     {targetBone {boneName=pelvis}}
    ]
    constraintWeights [1,1,1,1,1,1]
   }
   n66 = "Constraint - dyn_parent_pelvis_pack_r (PPPPPP)"!(
    ComponentPose=@n65
    Alpha=1
    ConstraintWeights_0=0.4413
    ConstraintWeights_1:float=0.1998
    ConstraintWeights_2:float=0.2212
    ConstraintWeights_3:float=0.0789
    ConstraintWeights_4:float=0.0243
    ConstraintWeights_5:float=0.0345
   ) -> (Pose) {
    alphaBoolBlend {blendOption=Linear}
    boneToModify {boneName=dyn_parent_pelvis_pack_r}
    constraintSetup [
     {targetBone {boneName=spine_01}}
     {targetBone {boneName=spine_02}}
     {targetBone {boneName=spine_03}}
     {targetBone {boneName=spine_04}}
     {targetBone {boneName=spine_05}}
     {targetBone {boneName=pelvis}}
    ]
    constraintWeights [1,1,1,1,1,1]
   }
   n67 = "Constraint - dyn_parent_pelvis_pack_c (PPPPPP)"!(
    ComponentPose=@n66
    Alpha=1
    ConstraintWeights_0=0.4475
    ConstraintWeights_1:float=0.2047
    ConstraintWeights_2:float=0.1963
    ConstraintWeights_3:float=0.0916
    ConstraintWeights_4:float=0.0333
    ConstraintWeights_5:float=0.0266
   ) -> (Pose) {
    alphaBoolBlend {blendOption=Linear}
    boneToModify {boneName=dyn_parent_pelvis_pack_c}
    constraintSetup [
     {targetBone {boneName=spine_01}}
     {targetBone {boneName=spine_02}}
     {targetBone {boneName=spine_03}}
     {targetBone {boneName=spine_04}}
     {targetBone {boneName=spine_05}}
     {targetBone {boneName=pelvis}}
    ]
    constraintWeights [1,1,1,1,1,1]
   }
   n68 = Flap Node - Bone: dyn_pelvis_pack_l!(ComponentPose:Component Space Pose Link Structure=@n67, Alpha:float=1) -> (Pose:Component Space Pose Link Structure) {
    is=title FortAnimGraphNode_Flap
    bUseClamp=true
    targetClampMax=3
    targetClampMin=0.25
    alphaBoolBlend {blendOption=Linear}
    sourceDrives [{
      bUseRange=false
      drivingCurve=null
      multiplier=1.1
      rangeMax=1
      rangeMin=-1
      remappedMax=1
      remappedMin=0
      sourceComponent=RotationZ
      useQuaternion=true
      sourceBone {boneName=thigh_l}
     }]
    targetBone {boneName=dyn_pelvis_pack_l}
   }
   n69 = Flap Node - Bone: dyn_pelvis_pack_r!(ComponentPose:Component Space Pose Link Structure=@n68, Alpha:float=1) -> (Pose:Component Space Pose Link Structure) {
    is=title FortAnimGraphNode_Flap
    bUseClamp=true
    targetClampMax=3
    alphaBoolBlend {blendOption=Linear}
    sourceDrives [
     {
      bUseRange=false
      drivingCurve=null
      multiplier=1.3
      rangeMax=1
      rangeMin=-1
      remappedMax=1
      remappedMin=0
      sourceComponent=RotationZ
      useQuaternion=true
      sourceBone {boneName=thigh_r}
     }
     {
      bUseRange=false
      drivingCurve=null
      multiplier=0.25
      rangeMax=1
      rangeMin=-1
      remappedMax=1
      remappedMin=0
      sourceComponent=RotationZ
      useQuaternion=true
      sourceBone {boneName=thigh_l}
     }
    ]
    targetBone {boneName=dyn_pelvis_pack_r}
   }
   n70 = Flap Node - Bone: dyn_pelvis_pack_c!(ComponentPose:Component Space Pose Link Structure=@n69, Alpha:float=1) -> (Pose:Component Space Pose Link Structure) {
    is=title FortAnimGraphNode_Flap
    bUseClamp=true
    targetClampMax=3
    alphaBoolBlend {blendOption=Linear}
    sourceDrives [
     {
      bUseRange=false
      drivingCurve=null
      multiplier=1
      rangeMax=1
      rangeMin=-1
      remappedMax=1
      remappedMin=0
      sourceComponent=RotationZ
      useQuaternion=true
      sourceBone {boneName=thigh_r}
     }
     {
      bUseRange=false
      drivingCurve=null
      multiplier=1.1
      rangeMax=1
      rangeMin=-1
      remappedMax=1
      remappedMin=0
      sourceComponent=RotationZ
      useQuaternion=true
      sourceBone {boneName=thigh_l}
     }
    ]
    targetBone {boneName=dyn_pelvis_pack_c}
   }
   n71 = "Constraint - dyn_upperarm_shell_01 (PPPPPP)"!(
    ComponentPose=@n70
    Alpha=1
    ConstraintWeights_0=0.1378
    ConstraintWeights_1:float=0.205
    ConstraintWeights_2:float=0.3283
    ConstraintWeights_3:float=0.0149
    ConstraintWeights_4:float=0.3021
    ConstraintWeights_5:float=0.0118
   ) -> (Pose) {
    alphaBoolBlend {blendOption=Linear}
    boneToModify {boneName=dyn_upperarm_shell_01}
    constraintSetup [
     {targetBone {boneName=clavicle_l}}
     {targetBone {boneName=upperarm_twist_02_l}}
     {targetBone {boneName=upperarm_twist_01_l}}
     {targetBone {boneName=lowerarm_l}}
     {targetBone {boneName=upperarm_l}}
     {targetBone {boneName=lowerarm_twist_02_l}}
    ]
    constraintWeights [1,1,1,1,1,1]
   }
   n72 = "Constraint - dyn_upperarm_shell_02 (PPPPPP)"!(
    ComponentPose=@n71
    Alpha=1
    ConstraintWeights_0=0.1409
    ConstraintWeights_1:float=0.1948
    ConstraintWeights_2:float=0.336
    ConstraintWeights_3:float=0.0181
    ConstraintWeights_4:float=0.2981
    ConstraintWeights_5:float=0.0121
   ) -> (Pose) {
    alphaBoolBlend {blendOption=Linear}
    boneToModify {boneName=dyn_upperarm_shell_02}
    constraintSetup [
     {targetBone {boneName=clavicle_l}}
     {targetBone {boneName=upperarm_twist_02_l}}
     {targetBone {boneName=upperarm_twist_01_l}}
     {targetBone {boneName=lowerarm_l}}
     {targetBone {boneName=upperarm_l}}
     {targetBone {boneName=lowerarm_twist_02_l}}
    ]
    constraintWeights [1,1,1,1,1,1]
   }
   n73 = "Constraint - dyn_upperarm_shell_03 (PPPPPP)"!(
    ComponentPose=@n72
    Alpha=1
    ConstraintWeights_0=0.1438
    ConstraintWeights_1:float=0.1896
    ConstraintWeights_2:float=0.3392
    ConstraintWeights_3:float=0.0196
    ConstraintWeights_4:float=0.2957
    ConstraintWeights_5:float=0.0121
   ) -> (Pose) {
    alphaBoolBlend {blendOption=Linear}
    boneToModify {boneName=dyn_upperarm_shell_03}
    constraintSetup [
     {targetBone {boneName=clavicle_l}}
     {targetBone {boneName=upperarm_twist_02_l}}
     {targetBone {boneName=upperarm_twist_01_l}}
     {targetBone {boneName=lowerarm_l}}
     {targetBone {boneName=upperarm_l}}
     {targetBone {boneName=lowerarm_twist_02_l}}
    ]
    constraintWeights [1,1,1,1,1,1]
   }
   n74 = "Constraint - dyn_upperarm_shell_04 (PPPPPPP)"!(
    ComponentPose=@n73
    Alpha=1
    ConstraintWeights_0=0.1434
    ConstraintWeights_1:float=0.1803
    ConstraintWeights_2:float=0.0102
    ConstraintWeights_3:float=0.3357
    ConstraintWeights_4:float=0.023
    ConstraintWeights_5:float=0.2965
    ConstraintWeights_6:float=0.011
   ) -> (Pose) {
    alphaBoolBlend {blendOption=Linear}
    boneToModify {boneName=dyn_upperarm_shell_04}
    constraintSetup [
     {targetBone {boneName=clavicle_l}}
     {targetBone {boneName=upperarm_twist_02_l}}
     {targetBone {boneName=spine_05}}
     {targetBone {boneName=upperarm_twist_01_l}}
     {targetBone {boneName=lowerarm_l}}
     {targetBone {boneName=upperarm_l}}
     {targetBone {boneName=lowerarm_twist_02_l}}
    ]
    constraintWeights [1,1,1,1,1,1,1]
   }
   n75 = "Constraint - dyn_upperarm_shell_05 (PPPPPP)"!(
    ComponentPose=@n74
    Alpha=1
    ConstraintWeights_0=0.1398
    ConstraintWeights_1:float=0.1821
    ConstraintWeights_2:float=0.3358
    ConstraintWeights_3:float=0.0238
    ConstraintWeights_4:float=0.0115
    ConstraintWeights_5:float=0.3071
   ) -> (Pose) {
    alphaBoolBlend {blendOption=Linear}
    boneToModify {boneName=dyn_upperarm_shell_05}
    constraintSetup [
     {targetBone {boneName=clavicle_l}}
     {targetBone {boneName=upperarm_twist_02_l}}
     {targetBone {boneName=upperarm_twist_01_l}}
     {targetBone {boneName=lowerarm_l}}
     {targetBone {boneName=dyn_upperarm_flap_01_l}}
     {targetBone {boneName=upperarm_l}}
    ]
    constraintWeights [1,1,1,1,1,1]
   }
   n76 = "Transform (Modify) Bone - Bone: dyn_upperarm_shell_01"(ComponentPose=@n75, Alpha=1, Translation="0.000,-0.632,-0.308", Rotation="0,0,0", Scale="1,1,1") {
    translationMode=BMM_Additive
    translationSpace=BCS_BoneSpace
    alphaBoolBlend {blendOption=Linear}
    boneToModify {boneName=dyn_upperarm_shell_01}
    translation {y=-0.632, z=-0.308}
   }
   n77 = "Transform (Modify) Bone - Bone: dyn_upperarm_shell_02"(ComponentPose=@n76, Alpha=1, Translation="0.000,-0.632,-0.308", Rotation="0,0,0", Scale="1,1,1") {
    translationMode=BMM_Additive
    translationSpace=BCS_BoneSpace
    alphaBoolBlend {blendOption=Linear}
    boneToModify {boneName=dyn_upperarm_shell_02}
    translation {y=-0.632, z=-0.308}
   }
   n78 = "Transform (Modify) Bone - Bone: dyn_upperarm_shell_03"(ComponentPose=@n77, Alpha=1, Translation="0.000,-0.632,-0.308", Rotation="0,0,0", Scale="1,1,1") {
    translationMode=BMM_Additive
    translationSpace=BCS_BoneSpace
    alphaBoolBlend {blendOption=Linear}
    boneToModify {boneName=dyn_upperarm_shell_03}
    translation {y=-0.632, z=-0.308}
   }
   n79 = "Transform (Modify) Bone - Bone: dyn_upperarm_shell_04"(ComponentPose=@n78, Alpha=1, Translation="0.000,-0.632,-0.308", Rotation="0,0,0", Scale="1,1,1") {
    translationMode=BMM_Additive
    translationSpace=BCS_BoneSpace
    alphaBoolBlend {blendOption=Linear}
    boneToModify {boneName=dyn_upperarm_shell_04}
    translation {y=-0.632, z=-0.308}
   }
   n80 = "Constraint - dyn_parent_flaps (PPPPP)"!(
    ComponentPose=@n79
    Alpha=1
    ConstraintWeights_0=0.0209
    ConstraintWeights_1:float=0.0338
    ConstraintWeights_2:float=0.243
    ConstraintWeights_3:float=0.6266
    ConstraintWeights_4:float=0.0758
   ) -> (Pose) {
    alphaBoolBlend {blendOption=Linear}
    boneToModify {boneName=dyn_parent_flaps}
    constraintSetup [
     {targetBone {boneName=spine_03}}
     {targetBone {boneName=spine_04}}
     {targetBone {boneName=spine_05}}
     {targetBone {boneName=clavicle_l}}
     {targetBone {boneName=upperarm_twist_01_l}}
    ]
    constraintWeights [1,1,1,1,1]
   }
   n81 = RigidBody(ComponentPose=@n80) {
    bClampLinearTranslationLimitToRefPose=true
    bOverrideWorldGravity=true
    overridePhysicsAsset=&/M_LRG_Tactical_Bear_Physics
    simulationSpace=BaseBoneSpace
    alphaBoolBlend {blendOption=Linear}
    baseBoneRef {boneName=root}
    componentAppliedLinearAccClamp=<1200 1200 1000>
    componentLinearAccScale=<0.7 0.7 0.3>
    componentLinearVelScale=<0.7 0.7 0.3>
    overrideWorldGravity {z=-700}
   }
   Output Pose(Result=Component To Local(ComponentPose=@n81)) {layerGroup=None}
   n82 = Flap Node - Bone: thigh_twist_offset_l(Alpha=1) {
    bUseClamp=true
    targetClampMin=-1.5
    alphaBoolBlend {blendOption=Linear}
    sourceDrives [{
      bUseRange=false
      drivingCurve=null
      multiplier=-0.25
      rangeMax=1
      rangeMin=-1
      remappedMax=1
      remappedMin=0
      sourceComponent=RotationX
      useQuaternion=true
      sourceBone {boneName=thigh_twist_01_l}
     }]
    targetBone {boneName=thigh_twist_offset_l}
   }
   Flap Node - Bone: thigh_twist_offset_r(ComponentPose=@n82, Alpha=1) {
    bUseClamp=true
    targetClampMax=1.5
    alphaBoolBlend {blendOption=Linear}
    sourceDrives [{
      bUseRange=false
      drivingCurve=null
      multiplier=0.25
      rangeMax=1
      rangeMin=-1
      remappedMax=1
      remappedMin=0
      sourceComponent=RotationX
      useQuaternion=true
      sourceBone {boneName=thigh_twist_01_r}
     }]
    targetBone {boneName=thigh_twist_offset_r}
   }
  ]}
 EventGraph {nodes [
   Event OnCharacterCustomizationLoaded()
   n1 = CheckForBackbling()
   > Set isWearingBackbling(isWearingBackbling=@n1)
   > Set isWearingCape(isWearingCape=@n1.bIsBackblingLong)
   > Branch(Condition=OR Boolean(A=Get isWearingCape() -> (isWearingCape:bool), B=Get isWearingBackbling()))
   > n2 = Get Dynamic Material Instance(SkeletalMeshComponent=GetPartSkeletalMeshComponent())
   > SetScalarParameterValue(self=@n2, ParameterName=Hide Element 10, Value=1)
  ]}
]
variables [
 isWearingBackbling {type=boolean}
 isWearingCape {type=boolean}
]
