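# Anim Blueprint for the Tactical Bear large male face accessory that adds secondary physics motion to its helmet flaps.
# It copies the attached parent mesh's pose, caches it, and feeds two AnimDynamics chains simulating left and right two-bone flaps with angular springs and collision limits.
# bIsParachuteOpened selects between the chains; the parachute-open chain adds clavicle sphere limits to the shared spine_05 sphere.
# The event graph has no logic.
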AnimBlueprint /BRCosmetics/Characters/Player/Male/Large/Bodies/M_LRG_Tactical_Bear/Meshes/Parts/M_LRG_Tactical_Bear_FaceAcc_AnimBP
target_skeleton=&/M_LRG_Tactical_Bear_FaceAcc_Skeleton
graphs [
 AnimGraph {nodes [
   n1 = Anim Dynamics - Bone: dyn_helmet_flap_01_l(ComponentPose=Local To Component(LocalPose=Use cached pose 'Start_CACHE'()), Alpha=1, ExternalForce="0,0,0") {
    angularDampingOverride=1
    angularSpringConstant=1500
    bAngularSpring=true
    bGravityOverrideInSimSpace=true
    bOverrideAngularDamping=true
    bUseGravityOverride=true
    gravityScale=0.35
    simulationSpace=RootRelative
    alphaBoolBlend {blendOption=Linear}
    boundBone {boneName=dyn_helmet_flap_01_l}
    componentAppliedLinearAccClamp=<3000 3000 3000>
    componentLinearAccScale=<0.3 0.3 0.3>
    componentLinearVelScale=<0.3 0.3 4>
    gravityOverride {z=-980}
    physicsBodyDefinitions [{
      sphereCollisionRadius=0
      boundBone {boneName=dyn_helmet_flap_01_l}
      boxExtents=<20 10 10>
      constraintSetup {angularLimitsMax=<5 5 25>, angularLimitsMin=<-5 -45 -25>, angularTarget=<0 0 0>}
      localJointOffset=<4 0 0>
     }]
   }
   n2 = Anim Dynamics - Bone: dyn_helmet_flap_02_l(ComponentPose=@n1, Alpha=1, ExternalForce="0,0,0") {
    angularDampingOverride=1
    angularSpringConstant=1500
    bAngularSpring=true
    bGravityOverrideInSimSpace=true
    bOverrideAngularDamping=true
    bUseGravityOverride=true
    bUseSphericalLimits=true
    gravityScale=0.35
    simulationSpace=RootRelative
    alphaBoolBlend {blendOption=Linear}
    boundBone {boneName=dyn_helmet_flap_02_l}
    componentAppliedLinearAccClamp=<3000 3000 3000>
    componentLinearAccScale=<0.3 0.3 0.3>
    componentLinearVelScale=<0.3 0.3 4>
    gravityOverride {z=-980}
    physicsBodyDefinitions [{
      collisionType=CustomSphere
      sphereCollisionRadius=1.8
      boundBone {boneName=dyn_helmet_flap_02_l}
      boxExtents=<10 10 10>
      constraintSetup {angularLimitsMax=<5 5 30>, angularLimitsMin=<-5 -45 -30>, angularTarget=<0 0 0>}
      localJointOffset=<4.5 0 0>
     }]
    planarLimits [{drivingBone {boneName=spine_05}, planeTransform {location=<30 -2 0>, rotation=<rot 0 0 90>, scale=<0.5 0.5 0.5>}}]
    sphericalLimits [
     {limitRadius=41.8, drivingBone {boneName=spine_05}, sphereLocalOffset=<-20 -15 0>}
     {limitRadius=19, drivingBone {boneName=clavicle_l}, sphereLocalOffset=<0 0 0>}
    ]
   }
   n3 = Anim Dynamics - Bone: dyn_helmet_flap_01_r(ComponentPose=@n2, Alpha=1, ExternalForce="0,0,0") {
    angularDampingOverride=1
    angularSpringConstant=1500
    bAngularSpring=true
    bGravityOverrideInSimSpace=true
    bOverrideAngularDamping=true
    bUseGravityOverride=true
    gravityScale=0.35
    simulationSpace=RootRelative
    alphaBoolBlend {blendOption=Linear}
    boundBone {boneName=dyn_helmet_flap_01_r}
    componentAppliedLinearAccClamp=<3000 3000 3000>
    componentLinearAccScale=<0.3 0.3 0.3>
    componentLinearVelScale=<0.3 0.3 4>
    gravityOverride {z=-980}
    physicsBodyDefinitions [{
      sphereCollisionRadius=0
      boundBone {boneName=dyn_helmet_flap_01_r}
      boxExtents=<20 10 10>
      constraintSetup {angularLimitsMax=<5 5 25>, angularLimitsMin=<-5 -45 -25>, angularTarget=<0 0 0>}
      localJointOffset=<-4 0 0>
     }]
   }
   n4 = Anim Dynamics - Bone: dyn_helmet_flap_02_r(ComponentPose=@n3, Alpha=1, ExternalForce="0,0,0") {
    angularDampingOverride=1
    angularSpringConstant=1500
    bAngularSpring=true
    bGravityOverrideInSimSpace=true
    bOverrideAngularDamping=true
    bUseGravityOverride=true
    bUseSphericalLimits=true
    gravityScale=0.35
    simulationSpace=RootRelative
    alphaBoolBlend {blendOption=Linear}
    boundBone {boneName=dyn_helmet_flap_02_r}
    componentAppliedLinearAccClamp=<3000 3000 3000>
    componentLinearAccScale=<0.3 0.3 0.3>
    componentLinearVelScale=<0.3 0.3 4>
    gravityOverride {z=-980}
    physicsBodyDefinitions [{
      collisionType=CustomSphere
      sphereCollisionRadius=1.9
      boundBone {boneName=dyn_helmet_flap_02_r}
      boxExtents=<10 10 10>
      constraintSetup {angularLimitsMax=<5 5 30>, angularLimitsMin=<-5 -45 -30>, angularTarget=<0 0 0>}
      localJointOffset=<-4 0 0>
     }]
    sphericalLimits [
     {limitRadius=41.8, drivingBone {boneName=spine_05}, sphereLocalOffset=<-20 -15 0>}
     {limitRadius=17.5, drivingBone {boneName=clavicle_r}, sphereLocalOffset=<0 0 0>}
    ]
   }
   n5 = Anim Dynamics - Bone: dyn_helmet_flap_01_l(ComponentPose=Local To Component(LocalPose=Use cached pose 'Start_CACHE'()), Alpha=1, ExternalForce="0,0,0") {
    angularDampingOverride=1
    angularSpringConstant=1500
    bAngularSpring=true
    bGravityOverrideInSimSpace=true
    bOverrideAngularDamping=true
    bUseGravityOverride=true
    gravityScale=0.35
    simulationSpace=RootRelative
    alphaBoolBlend {blendOption=Linear}
    boundBone {boneName=dyn_helmet_flap_01_l}
    componentAppliedLinearAccClamp=<3000 3000 3000>
    componentLinearAccScale=<0.3 0.3 0.3>
    componentLinearVelScale=<0.3 0.3 4>
    gravityOverride {z=-980}
    physicsBodyDefinitions [{
      sphereCollisionRadius=0
      boundBone {boneName=dyn_helmet_flap_01_l}
      boxExtents=<20 10 10>
      constraintSetup {angularLimitsMax=<5 5 25>, angularLimitsMin=<-5 -45 -25>, angularTarget=<0 0 0>}
      localJointOffset=<4 0 0>
     }]
   }
   n6 = Anim Dynamics - Bone: dyn_helmet_flap_02_l(ComponentPose=@n5, Alpha=1, ExternalForce="0,0,0") {
    angularDampingOverride=1
    angularSpringConstant=1500
    bAngularSpring=true
    bGravityOverrideInSimSpace=true
    bOverrideAngularDamping=true
    bUseGravityOverride=true
    bUseSphericalLimits=true
    gravityScale=0.35
    simulationSpace=RootRelative
    alphaBoolBlend {blendOption=Linear}
    boundBone {boneName=dyn_helmet_flap_02_l}
    componentAppliedLinearAccClamp=<3000 3000 3000>
    componentLinearAccScale=<0.3 0.3 0.3>
    componentLinearVelScale=<0.3 0.3 4>
    gravityOverride {z=-980}
    physicsBodyDefinitions [{
      collisionType=CustomSphere
      sphereCollisionRadius=1.8
      boundBone {boneName=dyn_helmet_flap_02_l}
      boxExtents=<10 10 10>
      constraintSetup {angularLimitsMax=<5 5 30>, angularLimitsMin=<-5 -45 -30>, angularTarget=<0 0 0>}
      localJointOffset=<4.5 0 0>
     }]
    planarLimits [{drivingBone {boneName=spine_05}, planeTransform {location=<30 -2 0>, rotation=<rot 0 0 90>, scale=<0.5 0.5 0.5>}}]
    sphericalLimits [{limitRadius=41.8, drivingBone {boneName=spine_05}, sphereLocalOffset=<-20 -15 0>}]
   }
   n7 = Anim Dynamics - Bone: dyn_helmet_flap_01_r(ComponentPose=@n6, Alpha=1, ExternalForce="0,0,0") {
    angularDampingOverride=1
    angularSpringConstant=1500
    bAngularSpring=true
    bGravityOverrideInSimSpace=true
    bOverrideAngularDamping=true
    bUseGravityOverride=true
    gravityScale=0.35
    simulationSpace=RootRelative
    alphaBoolBlend {blendOption=Linear}
    boundBone {boneName=dyn_helmet_flap_01_r}
    componentAppliedLinearAccClamp=<3000 3000 3000>
    componentLinearAccScale=<0.3 0.3 0.3>
    componentLinearVelScale=<0.3 0.3 4>
    gravityOverride {z=-980}
    physicsBodyDefinitions [{
      sphereCollisionRadius=0
      boundBone {boneName=dyn_helmet_flap_01_r}
      boxExtents=<20 10 10>
      constraintSetup {angularLimitsMax=<5 5 25>, angularLimitsMin=<-5 -45 -25>, angularTarget=<0 0 0>}
      localJointOffset=<-4 0 0>
     }]
   }
   n8 = Anim Dynamics - Bone: dyn_helmet_flap_02_r(ComponentPose=@n7, Alpha=1, ExternalForce="0,0,0") {
    angularDampingOverride=1
    angularSpringConstant=1500
    bAngularSpring=true
    bGravityOverrideInSimSpace=true
    bOverrideAngularDamping=true
    bUseGravityOverride=true
    bUseSphericalLimits=true
    gravityScale=0.35
    simulationSpace=RootRelative
    alphaBoolBlend {blendOption=Linear}
    boundBone {boneName=dyn_helmet_flap_02_r}
    componentAppliedLinearAccClamp=<3000 3000 3000>
    componentLinearAccScale=<0.3 0.3 0.3>
    componentLinearVelScale=<0.3 0.3 4>
    gravityOverride {z=-980}
    physicsBodyDefinitions [{
      collisionType=CustomSphere
      sphereCollisionRadius=1.9
      boundBone {boneName=dyn_helmet_flap_02_r}
      boxExtents=<10 10 10>
      constraintSetup {angularLimitsMax=<5 5 30>, angularLimitsMin=<-5 -45 -30>, angularTarget=<0 0 0>}
      localJointOffset=<-4 0 0>
     }]
    sphericalLimits [{limitRadius=41.8, drivingBone {boneName=spine_05}, sphereLocalOffset=<-20 -15 0>}]
   }
   Output Pose(Result=Blend Poses by bool(BlendPose_0=Component To Local(ComponentPose=@n4), BlendPose_1=Component To Local(ComponentPose=@n8), bActiveValue=Get bIsParachuteOpened())) {layerGroup=None}
   Save cached pose 'Start_CACHE'(Pose=Copy Pose From Mesh() {bUseAttachedParent=true}) {cacheName=Start_CACHE}
  ]}
 EventGraph {nodes [
   TryGetPawnOwner()
   Event BlueprintUpdateAnimation()
  ]}
]
variables []
